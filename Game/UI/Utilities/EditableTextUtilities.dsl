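; UI utility functions that cap the character length of text typed into editable text widgets.
; LimitTextLength casts the given widget to EditableText, falling back to EditableTextBox when that cast fails, and reads its current text.
; CheckIfNeedsTrimming returns the left-truncated text and an over-limit flag.
; LimitTextLength writes the truncated text back only when over the limit, and returns whether it trimmed.
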(fn CheckIfNeedsTrimming
  (params (Text :type Text) (Limit :type Integer))
  (return (Utilities|Text|ToText(String) (Utilities|String|Left (Utilities|String|ToString(Text) Text) Limit)) (> (Utilities|String|Len (Utilities|String|ToString(Text) Text)) Limit)))

(fn LimitTextLength
  (params (EditableTextWidget :type Widget%20Object%20Reference) (Length :type Integer))
  (bind _aseditable_text (Utilities|Casting|CastToEditableText EditableTextWidget)
    (:then
      (bind (_trimmedtext _wastrimmed) (CallFunction|CheckifNeedsTrimming (Widget|GetText(EditableText) _aseditable_text) Length _aseditable_text))
      (if _wastrimmed
        (Widget|SetText(EditableText) _aseditable_text _trimmedtext)
        (else
          _))
      (return _wastrimmed))
    (:CastFailed
      (bind _astext_box (Utilities|Casting|CastToEditableTextBox EditableTextWidget))
      (bind (_trimmedtext_2 _wastrimmed_2) (CallFunction|CheckifNeedsTrimming (Widget|GetText(TextBox) _astext_box) Length _astext_box))
      (if _wastrimmed_2
        (Widget|SetText(TextBox) _astext_box _trimmedtext_2)
        (else
          _))
      (return _wastrimmed_2))))
